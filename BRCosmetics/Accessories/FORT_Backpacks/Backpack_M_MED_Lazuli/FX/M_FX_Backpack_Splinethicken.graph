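# Translucent unlit Niagara mesh-particle material for the Lazuli backpack, drawing glowing, wavy two-color strands widened by SplineThicken using vertex color.
# Wave-distorted UVs divided by a noise-varied width form the strand mask, with dynamic parameters controlling hide, color offset, wave offset and width.
# Animated color split and pulsing brightness feed emission, overlaid with optional noise-broken sine stripes.
# Opacity combines the mask, particle alpha and depth fade, skipping the fade on forward and mobile.
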Material /BRCosmetics/Accessories/FORT_Backpacks/Backpack_M_MED_Lazuli/FX/M_FX_Backpack_Splinethicken
graph {
 properties {
  bUsedWithNiagaraMeshParticles=true
  blendMode=BLEND_Translucent
  pixelDepthOffsetMode=PDOM_Legacy
  shadingModel=MSM_Unlit
  usedShadingModels=Unlit
 }
 nodes [
  n1 = Constant() {r=0.5}
  n2 = Add(A=@n1, B=Constant())
  n3 = NamedRerouteUsage()
  Add(A=NamedRerouteUsage(), B=Multiply(A=SmoothStep(Min=@n1, Max=@n2, Value=@n3), B=NamedRerouteUsage()))
  n4 = Multiply(A=VertexColor().B, B=ScalarParameter() {defaultValue=6.5, parameterName=SplineWidth})
  MaterialFunctionCall(WidthBase=@n4, WidthTip=@n4, Expand U or V UV Channel=StaticBool() {value=true}) {materialFunction=&SplineThicken}
  n5 = Constant() {r=0.6}
  n6 = Constant() {r=0.1}
  n7 = ComponentMask(None=TextureCoordinate()) {g=true}
  n8 = TextureCoordinate()
  n9 = ParticleRandom()
  n10 = Constant() {r=0.2}
  n11 = Constant()
  n12 = TextureSample(UVs=Multiply(
    A=Add(A=@n8, B=@n9)
    B=MaterialFunctionCall(X=@n10, Y=@n11) {materialFunction=&MakeFloat2}.Result
   ) {constB=5}) {samplerType=SAMPLERTYPE_Grayscale, texture=&LowResBlurredNoise}
  n13 = Sine(None=Multiply(A=Add(A=@n7, B=@n12.R), B=ScalarParameter() {defaultValue=60, parameterName=Stripes Amount}))
  n14 = Multiply(A=SmoothStep(Min=@n5, Max=Add(A=@n5, B=@n6), Value=ConstantBiasScale(None=@n13)), B=NamedRerouteUsage())
  n15 = NamedRerouteUsage()
  NamedRerouteDeclaration(None=StaticSwitchParameter(True=Multiply(A=@n14, B=@n15), False=Constant()) {defaultValue=true, parameterName=UseStripes}) {name=Stripes, nodeColor=<rgba 0 0.2031 1 1>}
  n16 = MaterialFunctionCall(Float2=TextureCoordinate()) {materialFunction=&BreakOutFloat2Components}
  n17 = OneMinus(None=@n16.R)
  n18 = DynamicParameter() {defaultValue=<rgba 0 1 0 0.25>, paramNames [Hide, Color Offset, Wave Offset, Width]}
  n19 = Add(
   A=MaterialFunctionCall(X=LinearInterpolate(A=@n16.R, B=@n17, Alpha=@n18.Wave Offset), Y=@n16.G) {materialFunction=&MakeFloat2}.Result
   B=MaterialFunctionCall(X=Constant(), Y=ScalarParameter()) {materialFunction=&MakeFloat2}.Result
  )
  n20 = ScalarParameter() {defaultValue=4, parameterName=WaveFreq}
  n21 = ScalarParameter() {defaultValue=0.05, parameterName=WaveAmp}
  n22 = Constant() {r=1}
  n23 = Time()
  n24 = ConstantBiasScale(None=ComponentMask(None=MaterialFunctionCall!(UV=@n19, U Frequency, V Frequency=@n20, U Amplitude, V Amplitude=@n21, U Speed, V Speed=@n22, U Time=@n23, V Time=@n23) {materialFunction=/Game/Effects/Fort_Effects/MaterialFunctions/MF_WaveUV}.Wave UVs) {r=true}) {bias=-0.5, scale=2}
  n25 = DynamicParameter() {defaultValue=<rgba 0 1 0 0.25>, paramNames [Hide, Color Offset, Wave Offset, Width]}
  n26 = Add(A=@n25.Hide, B=Constant() {r=0.6})
  n27 = MaterialFunctionCall!(Gradient=ComponentMask(None=TextureCoordinate()) {g=true}, Center Point=Constant() {r=0.35}) {materialFunction=/Game/Effects/Fort_Effects/MaterialFunctions/MF_GradientReflected}
  n28 = DynamicParameter() {defaultValue=<rgba 0 1 0 0.25>, paramNames [Hide, Color Offset, Wave Offset, Width]}
  n29 = Constant()
  n30 = ParticleRandom()
  n31 = Panner(Coordinate=Add(
    A=TextureCoordinate()
    B=MaterialFunctionCall(X=@n29, Y=@n30) {materialFunction=&MakeFloat2}.Result
   )) {speedY=0.2}
  n32 = MaterialFunctionCall(X=Constant(), Y=Constant() {r=1}) {materialFunction=&MakeFloat2}
  n33 = LinearInterpolate(Alpha=TextureSample(UVs=Multiply(A=@n31, B=@n32.Result)) {samplerType=SAMPLERTYPE_Grayscale, texture=/FortWater/Textures/LowResBlurredNoise}.R) {constA=0.5, constB=1.5}
  n34 = Divide(A=@n24, B=Multiply(A=SmoothStep(Min=@n25.Hide, Max=@n26, Value=@n27.Result), B=Multiply(A=@n28.Width, B=@n33)))
  n35 = OneMinus(None=Saturate(None=DotProduct(A=@n34, B=@n34)))
  n36 = Constant() {r=0.7}
  NamedRerouteDeclaration(None=OneMinus(None=Step(Y=@n35, X=@n36) {constX=0.2})) {name=Mask, nodeColor=<rgba 0.6406 1 0 1>}
  n37 = NamedRerouteUsage()
  n38 = ParticleColor()
  n39 = Power(Base=DepthFade(FadeDistance=ScalarParameter() {defaultValue=3, parameterName=FadeDistance})) {constExponent=3}
  n40 = Constant() {r=1}
  NamedRerouteDeclaration(None=Multiply(A=Multiply(A=@n37, B=@n38.A), B=ShadingPathSwitch(Default=@n39, Forward=@n40, Mobile=@n40))) {name=Opacity, nodeColor=<rgba 0.9922 0 1 1>}
  n41 = Constant() {r=0.5}
  n42 = Constant() {r=0.2}
  n43 = Constant() {r=0.5}
  n44 = SphereMask(A=ComponentMask(None=TextureCoordinate()) {g=true}, B=@n43, Radius=@n43) {hardnessPercent=0}
  NamedRerouteDeclaration(None=SmoothStep(Min=@n41, Max=Add(A=@n41, B=@n42), Value=OneMinus(None=@n44))) {name=VerticalMask, nodeColor=<rgba 0 1 0.9531 1>}
  n45 = TextureCoordinate()
  n46 = Add(A=Multiply(A=ComponentMask(None=@n45) {g=true}, B=Constant() {r=3}), B=Multiply(A=Time(), B=Constant() {r=0.5}))
  n47 = ParticleRandom()
  n48 = ConstantBiasScale(None=Sine(None=Add(A=@n46, B=@n47)))
  n49 = DynamicParameter() {defaultValue=<rgba 0 1 0 0.25>, paramNames [Hide, Color Offset, Wave Offset, Width]}
  n50 = Step(Y=@n24, X=LinearInterpolate(Alpha=Add(A=@n48, B=@n49.Color Offset) {constB=0}) {constB=-0.15}) {constX=-0.1}
  n51 = LinearInterpolate(
   A=Constant3Vector() {constant=<rgba 0.0353 0 0.3569 1>}
   B=Constant3Vector() {constant=<rgba 0.9473 0.0012 0.0742 1>}
   Alpha=LinearInterpolate(A=@n50, B=OneMinus(None=@n50), Alpha=ScalarParameter() {parameterName=Switch Colors?})
  )
  n52 = Constant() {r=0.5}
  n53 = SphereMask(A=ComponentMask(None=TextureCoordinate()) {g=true}, B=@n52, Radius=@n52) {hardnessPercent=0}
  n54 = LinearInterpolate(Alpha=Power(Base=OneMinus(None=@n53))) {constA=1, constB=6}
  n55 = Sine(None=Multiply(A=Time(), B=Constant() {r=3}))
  n56 = Multiply(A=Multiply(A=Multiply(A=@n51, B=@n54), B=@n38.RGB), B=LinearInterpolate(Alpha=ConstantBiasScale(None=@n55)) {constA=0.5})
  n57 = VectorParameter() {parameterName=StripesColor, defaultValue=<rgba 0.0353 0 0.3569 1>}
  n58 = Constant() {r=0.3}
  n59 = SmoothStep(Min=@n58, Max=Add(A=@n58, B=Constant() {r=0.2}), Value=NamedRerouteUsage())
  NamedRerouteDeclaration(None=MaterialFunctionCall(Emissive=LinearInterpolate(A=@n56, B=@n57.RGB, Alpha=@n59)) {materialFunction=&MF_FX_EmissiveAdjustments}.Result) {name=Emission, nodeColor=<rgba 0.0078 0 1 1>}
  n60 = Constant()
  NamedRerouteDeclaration(None=OneMinus(None=Step(Y=@n35, X=@n60) {constX=0})) {name=StripesMask, nodeColor=<rgba 1 0.3984 0 1>}
  n61 = Constant() {r=0.5}
  n62 = Constant()
  n63 = ComponentMask(None=TextureCoordinate()) {r=true}
  n64 = OneMinus(None=@n63)
  n65 = ScalarParameter() {parameterName=SwapStripesHor?}
  NamedRerouteDeclaration(None=SmoothStep(Min=@n61, Max=Add(A=@n61, B=@n62), Value=LinearInterpolate(A=@n63, B=@n64, Alpha=@n65))) {name=HorizontalMask, nodeColor=<rgba 0 1 0.9531 1>}
  NamedRerouteUsage()
 ]
}
