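# Niagara effect for the Lazuli backpack cosmetic, adding flickering purple lights and animated mesh waves.
# The Light emitter bursts two light particles at mirrored side offsets, pulsing brightness with a curve over emitter loop age.
# Waves_L and Waves_R continuously spawn meshes drifting in opposite X directions, animated through dynamic material parameters.
# User.Visibility, defaulting to zero, scales light color and wave opacity, while time-of-day scalars dim both.
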NiagaraSystem /BRCosmetics/Accessories/FORT_Backpacks/Backpack_M_MED_Lazuli/FX/NS_Backpack_Lazuli
dependencies {
 usedDataInterfaces [&NiagaraDataInterfaceCurve, &NiagaraDataInterfaceArrayFloat3]
 usedDynamicInputs []
 usedModules [&SystemState]
}
emitters {
 Light {
  emitterUpdateScript [
   &EmitterState(Inactive Response=::NewEnumerator1, Life Cycle Mode=::NewEnumerator1, Loop Duration=2, UseLoopDelay=true)
   &SpawnBurst_Instantaneous(Spawn Count=2, Use Spawn Probability=false)
   /SetVariables(Emitter.LightsActive:NiagaraBool=&SetBoolByFloatComparison(A=1, B=$User.Visibility, Comparison Type=::NewEnumerator4)) {isSetParameters=true}
  ]
  particleSpawnScript [
   &FN_SetSpawnGroup(Number of Spawn Groups=2)
   &InitializeParticle(
    Color Mode=::NewEnumerator0
    Lifetime=4
    Position Offset=<20 -20 0>
    Sprite Size Mode=::NewEnumerator3
    Uniform Sprite Size=5
    UsePositionOffset=false
    Write Material Random=true
   ) {module=InitializeParticle001}
   &SystemLocation(
    MaskSpawnGroup=false
    Offset=&SelectVectorFromArray(
     Direct Array Index=$Particles.SpawnGroup
     Vector Selection Array=DataInterface() {propertyValues {FloatData [<17.5 -21 -1>, <-17.5 -21 -1>]}}
    )
   ) {module=SystemLocation002}
  ]
  particleUpdateScript [
   &ParticleState()
   &Color(Color=&MakeLinearColorFromVectorAndFloat(
     "Vector (RGB)"=&Multiply_Vector(
      A=&MakeVectorFromLinearColorRGB(Color=<rgba 0.4956 0.0237 1 1>)
      B=&VectorFromFloat(Value=&Multiply_Float(
        A=&FloatFromCurve(
         FloatCurve=DataInterface() {propertyValues {Curve {keys [
             <at 0 1 RCIM_Cubic>
             {arriveTangent=1.216, interpMode=RCIM_Cubic, leaveTangent=1.216, time=0.1, value=0.573}
             {arriveTangent=2.2985, interpMode=RCIM_Cubic, leaveTangent=2.2985, time=0.1, value=1.1216}
             {arriveTangent=-4.608, interpMode=RCIM_Cubic, leaveTangent=-4.608, time=0.3, value=1.0327}
             {arriveTangent=-4.4434, interpMode=RCIM_Cubic, leaveTangent=-4.4434, time=0.3, value=0.2}
             {arriveTangent=2.9184, interpMode=RCIM_Cubic, leaveTangent=2.9184, time=0.4, value=0.5884}
             {arriveTangent=-1.6918, interpMode=RCIM_Cubic, leaveTangent=-1.6918, time=0.5, value=0.7837}
             {arriveTangent=1.7608, interpMode=RCIM_Cubic, leaveTangent=1.7608, time=0.6, value=0.25}
             {arriveTangent=2.9444, interpMode=RCIM_Cubic, leaveTangent=2.9444, time=0.7, value=1.1358}
             {arriveTangent=-1.3526, interpMode=RCIM_Cubic, leaveTangent=-1.3526, time=0.7, value=0.5444}
             {arriveTangent=-1.7222, interpMode=RCIM_Cubic, leaveTangent=-1.7222, time=0.8, value=1.0006}
             {arriveTangent=-1.2799, interpMode=RCIM_Cubic, leaveTangent=-1.2799, time=0.9, value=0.2}
             {interpMode=RCIM_Cubic, leaveTangent=-4.6792, time=1, value=0.7446}
            ]}}}
         CurveIndex=$Emitter.NormalizedLoopAge
         Scale Curve=&FNTimeOfDayScalar(Night Scale=0.7, Morning Scale=0.6, Evening Scale=0.6)
        )
        B=$User.Visibility
       ))
     )
     "Float (Alpha)"=1
    ))
   &FN_Light_Attributes(
    Color=$Particles.Color
    Create Color Attribute=true
    Create Exponent Attribute=true
    Create Radius Attribute=true
    Enable Particle Light Enable=true
    Exponent=1.5
    Radius=18
    Scale Color=2
   )
  ]
  rendererClasses [&NiagaraLightRendererProperties]
  renderers [{rendererClass=&NiagaraLightRendererProperties, rendererIndex=0}]
 }
 Waves_L {
  emitterUpdateScript [
   &EmitterState()
   &SpawnRate(SpawnRate=2, Use Spawn Probability=false)
  ]
  particleSpawnScript [
   &InitializeParticle(Color Mode=::NewEnumerator0, Lifetime=2, Mesh Scale Mode=::NewEnumerator4, UsePositionOffset=false, Write Material Random=true)
   &AddVelocity(Velocity=<6 -0.5 0>)
   &SystemLocation(MaskSpawnGroup=false, Offset=<0 0 3>)
  ]
  particleUpdateScript [
   &ParticleState()
   &ScaleMeshSize(Scale Factor=&VectorFromFloat(Value=0.8))
   &SolveForcesAndVelocity(Clamp Velocity=false, Limit Acceleration=false)
   &DynamicMaterialParameters(
    Index 0 Param 1=&FloatFromCurve(FloatCurve=DataInterface() {propertyValues {Curve {keys [<at 0 0>,<at 1 0.7>]}}})
    Index 0 Param 2=&Add_Float(
     A=&FloatFromCurve(FloatCurve=DataInterface() {propertyValues {Curve {keys [<at 0 -0.8599>,<at 1 -1.7599>]}}})
     B=0.5
    )
    Index 0 Param 4=&FloatFromCurve(
     FloatCurve=DataInterface() {propertyValues {Curve {keys [
         <at 0 0 RCIM_Cubic>
         {arriveTangent=0.375, interpMode=RCIM_Cubic, leaveTangent=0.375, time=0.15, value=0.35}
         {arriveTangent=-0.4118, interpMode=RCIM_Cubic, leaveTangent=-0.4118, time=0.8, value=0.3}
         <at 1 0 RCIM_Cubic>
        ]}}}
     Scale Curve=&Multiply_Float(A=1.25, B=$User.Visibility)
    )
    Module.Index0Param0Write=true
    Module.Index0Param1Write=true
    Module.Index0Param2Write=true
    Module.Index0Param3Write=true
   )
   &ScaleColor(
    Scale Alpha=&FloatFromCurve(FloatCurve=DataInterface() {propertyValues {Curve {keys [<at 0 0 RCIM_Cubic>,<at 0.2 1 RCIM_Cubic>]}}}, Scale Curve=$User.Visibility)
    Scale RGB=&VectorFromFloat(Value=&FNTimeOfDayScalar(Night Scale=0.7, Morning Scale=0.8, Evening Scale=0.8))
    ScaleA=true
    ScaleRGB=true
   )
  ]
  rendererClasses [&NiagaraMeshRendererProperties]
  renderers [{rendererClass=&NiagaraMeshRendererProperties, rendererIndex=0}]
 }
 Waves_R {
  emitterUpdateScript [
   &EmitterState()
   &SpawnRate(SpawnRate=2, Use Spawn Probability=false)
  ]
  particleSpawnScript [
   &InitializeParticle(Color Mode=::NewEnumerator0, Lifetime=2, Mesh Scale Mode=::NewEnumerator4, UsePositionOffset=false, Write Material Random=true)
   &AddVelocity(Velocity=<-6 -0.5 0>)
   &SystemLocation(MaskSpawnGroup=false, Offset=<0 0 3>)
  ]
  particleUpdateScript [
   &ParticleState()
   &ScaleMeshSize(Scale Factor=&VectorFromFloat(Value=0.8))
   &SolveForcesAndVelocity(Clamp Velocity=false, Limit Acceleration=false)
   &DynamicMaterialParameters(
    Index 0 Param 1=&FloatFromCurve(FloatCurve=DataInterface() {propertyValues {Curve {keys [<at 0 0>,<at 1 0.7>]}}})
    Index 0 Param 2=&Add_Float(
     A=&FloatFromCurve(FloatCurve=DataInterface() {propertyValues {Curve {keys [<at 0 -0.875>,<at 1 -1.75>]}}})
     B=0.5
    )
    Index 0 Param 3=1
    Index 0 Param 4=&FloatFromCurve(
     FloatCurve=DataInterface() {propertyValues {Curve {keys [
         <at 0 0 RCIM_Cubic>
         {arriveTangent=0.375, interpMode=RCIM_Cubic, leaveTangent=0.375, time=0.15, value=0.35}
         {arriveTangent=-0.4118, interpMode=RCIM_Cubic, leaveTangent=-0.4118, time=0.8, value=0.3}
         <at 1 0 RCIM_Cubic>
        ]}}}
     Scale Curve=&Multiply_Float(A=1.25, B=$User.Visibility)
    )
    Module.Index0Param0Write=true
    Module.Index0Param1Write=true
    Module.Index0Param2Write=true
    Module.Index0Param3Write=true
   )
   &ScaleColor(
    Scale Alpha=&FloatFromCurve(FloatCurve=DataInterface() {propertyValues {Curve {keys [<at 0 0 RCIM_Cubic>,<at 0.2 1 RCIM_Cubic>]}}}, Scale Curve=$User.Visibility)
    Scale RGB=&VectorFromFloat(Value=&FNTimeOfDayScalar(Night Scale=0.7, Morning Scale=0.8, Evening Scale=0.8))
    ScaleA=true
    ScaleRGB=true
   )
  ]
  rendererClasses [&NiagaraMeshRendererProperties]
  renderers [{rendererClass=&NiagaraMeshRendererProperties, rendererIndex=0}]
 }
}
summary {
 systemName=NS_Backpack_Lazuli
 emitters [
  {emitterName=Waves_L, rendererClasses [&NiagaraMeshRendererProperties]}
  {emitterName=Light, rendererClasses [&NiagaraLightRendererProperties]}
  {emitterName=Waves_R, rendererClasses [&NiagaraMeshRendererProperties]}
 ]
 userVariables [
  User.bIsFrontEnd {type=NiagaraBool}
  User.bIsFrontEndPreview {type=NiagaraBool}
  User.Visibility {type=NiagaraFloat, defaultValue=0, description=""}
 ]
}
