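# Animation Blueprint for the Lazuli backpack cosmetic, whose jaw reacts when the player eliminates an enemy.
# The pack bone aims at the owner mesh's head socket, and the root offsets by an alpha from spine bone deltas.
# That cached pose feeds either spring-driven AnimDynamics jaw jiggle or a fixed jaw rotation, blended by bIsActiveKill.
# The EnemyPlayerEliminated anim input event sets that flag, and a one-second retriggerable delay clears it.
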AnimBlueprint /BRCosmetics/Accessories/FORT_Backpacks/Backpack_M_MED_Lazuli/Meshes/M_MED_Lazuli_Pack_AnimBP
target_skeleton=/Game/Accessories/FORT_Backpacks/Backpack_M_MED_Lazuli/Meshes/M_MED_Lazuli_Pack_Skeleton
graphs [
 AnimGraph {nodes [
   n1 = Anim Dynamics - Bone: dyn_jaw(ComponentPose=Local To Component(LocalPose=Use cached pose 'InitialPose'()), Alpha=1, ExternalForce="0,0,0") {
    angularDampingOverride=0.98
    angularSpringConstant=1.5
    bAngularSpring=true
    bOverrideAngularDamping=true
    bUseGravityOverride=true
    alphaBoolBlend {blendOption=Linear}
    boundBone {boneName=dyn_jaw}
    componentAppliedLinearAccClamp=<120 120 120>
    componentLinearAccScale=<1 1 1>
    componentLinearVelScale=<0.5 0.5 0.5>
    gravityOverride {z=-980}
    physicsBodyDefinitions [{
      boundBone {boneName=dyn_jaw}
      boxExtents=<15 10 15>
      constraintSetup {angularLimitsMax=<2.5 0 2.5>, angularLimitsMin=<-2.5 0 -2.5>, angularTarget=<0 0 0>}
      localJointOffset=<0 -4 0>
     }]
   }
   n2 = "Transform (Modify) Bone - Bone: dyn_jaw"!(ComponentPose=Local To Component(LocalPose=Use cached pose 'InitialPose'()), Alpha=1, Rotation="0,0,-40") -> (Pose) {
    rotationMode=BMM_Additive
    rotationSpace=BCS_BoneSpace
    alphaBoolBlend {blendInTime=0.3, blendOption=ExpOut, blendOutTime=0.2}
    boneToModify {boneName=dyn_jaw}
    rotation {roll=-40}
   }
   Output Pose(Result=Blend Poses by bool(
     BlendPose_0=Component To Local(ComponentPose=@n2)
     BlendPose_1=Component To Local(ComponentPose=@n1)
     BlendTime_0=0.2
     bActiveValue=Get bIsActiveKill() -> (bIsActiveKill:bool)
    ) {blendType=Linear}) {layerGroup=None}
   n3 = Input Pose()
   n4 = "Transform (Modify) Bone - Bone: pack"(
    ComponentPose=Look At - Bone: pack!(ComponentPose=Local To Component(LocalPose=@n3), Alpha=1, LookAtLocation:Vector=Get HeadLocation() -> (HeadLocation:Vector)) -> (Pose) {alphaBoolBlend {blendOption=Linear}, boneToModify {boneName=pack}, lookAt_Axis {axis {y=0, z=1}}, lookAtTarget=~}
    Translation="0,0,0"
    Rotation="0,0,-45"
   ) {
    rotationMode=BMM_Additive
    rotationSpace=BCS_BoneSpace
    translationMode=BMM_Additive
    alphaBoolBlend {blendOption=Linear}
    boneToModify {boneName=pack}
    rotation {roll=-45}
   }
   n5 = "Transform (Modify) Bone - Bone: root"(ComponentPose=@n4, Alpha=Get SpineRotationAlpha() -> (SpineRotationAlpha:float), Translation="-3,-1,0", Rotation="0,-20,0") {
    rotationMode=BMM_Additive
    rotationSpace=BCS_ParentBoneSpace
    translationMode=BMM_Additive
    translationSpace=BCS_ParentBoneSpace
    alphaBoolBlend {blendOption=Linear}
    alphaScaleBiasClamp {bInterpResult=true}
    boneToModify {boneName=root}
    rotation {yaw=-20}
    translation=<-3 -1>
   }
   Save cached pose 'InitialPose'(Pose=Component To Local(ComponentPose=@n5)) {cacheName=InitialPose}
  ]}
 EventGraph {nodes [
   Event BlueprintUpdateAnimation()
   > Set SpineRotationAlpha(SpineRotationAlpha:float=GetAlphaFromReferenceTransformDelta(
     SkeletalMeshComponent=Get Skel Mesh()
     BoneName=spine_03
     BaseName=spine_05
     DeltaComponentType=EulerX
     InputRange=(X=0,Y=10)
     TwistAxis="1,0,0"
    )) -> (Output_Get:float)
   n1 = Get Skel Mesh()
   > Is Valid(InputObject=@n1)
   > Set HeadLocation(HeadLocation:Vector=GetSocketLocation!(self:Scene Component Object Reference=@n1, InSocketName:Name=head) -> (ReturnValue:Vector)) -> (Output_Get:Vector)
   Event BlueprintInitializeAnimation()
   > Set Skel Mesh(Skel Mesh=GetOwnerSkeletalMeshComponent())
   n2 = Event BlueprintProcessAnimInputEvent()
   > Branch(Condition=IsNamed(self=@n2.AnimInputEvent, InEventName=aie.EnemyPlayerEliminated))
   > n3 = Branch(Condition=Get bIsActiveKill())
   RetriggerableDelay(execute=[@n3.then, @n4.then], Duration=1)
   > Set bIsActiveKill(bIsActiveKill:bool=false) -> (Output_Get:bool)
   n4 = Set bIsActiveKill(execute=@n3.else, bIsActiveKill=true)
  ]}
]
variables [
 bIsActiveKill {type=boolean}
 bodyMesh {type=SkeletalMeshComponent}
 HeadLocation {type=Vector}
 Skel Mesh {type=SkeletalMeshComponent}
 SpineRotationAlpha {type=number}
]
